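# Unlit, translucent, two-sided material for mesh particles, named as a blocking ground effect in the Hightower_QJ set.
# It builds a procedural, texture-free band along the V coordinate: cubing and halving V, then applying a sine, skews the peak toward high V and fades both ends.
# Particle color and alpha multiply that band, and the result feeds the emissive input of MF_FX_EmissiveAdjustments.
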Material /Game/Environments/Apollo/Sets/Hightower_QJ/FXs/FX_Materials/M_GroundEffect_Blocking
graph {
 properties {
  bUsedWithMeshParticles=true
  blendMode=BLEND_Translucent
  pixelDepthOffsetMode=PDOM_Legacy
  shadingModel=MSM_Unlit
  twoSided=true
  usedShadingModels=Unlit
 }
 nodes [
  n1 = ParticleColor()
  n2 = ComponentMask(None=TextureCoordinate()) {g=true}
  n3 = Sine(None=Multiply(A=Power(Base=@n2) {constExponent=3}) {constB=0.5})
  MaterialFunctionCall(Emissive=Multiply(A=@n1.RGB, B=Multiply(A=@n1.A, B=@n3))) {materialFunction=&MF_FX_EmissiveAdjustments}
 ]
}
